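# Small cloth flag flapping ambient loop cue for a Ruined Reels POI spot, reusing the Capture the Flag flag loop wave.
# A random delay of up to two seconds staggers the start of the looping wave, which a mixer boosts.
# Two nested, endlessly looping envelopers add variation: a 20-second volume swell and a 60-second pitch drift.
# The cue plays pitched down, capped at two voices, with shared flag attenuation and a reverb send.
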SoundCue /Game/Sounds/Helios/POI/RuinedReels/Spots/Spot_Cloth_Flag_Flap_Small_Loop_Cue
root=Delay_0
nodes [
 {id=Delay_0, children [Enveloper_2], properties {delayMax=2}}
 {
  id=Enveloper_1
  children [Mixer_0]
  properties {
   bLoop=true
   bLoopIndefinitely=true
   loopEnd=60
   pitchCurve {editorCurveData {keys [
      {interpMode=RCIM_Cubic, time=0}
      <at 15 0.95 RCIM_Cubic>
      {arriveTangent=0.0072, interpMode=RCIM_Cubic, leaveTangent=0.0072, tangentMode=RCTM_User, time=30}
      <at 45 1.1 RCIM_Cubic>
      {time=60}
     ]}}
   volumeCurve {editorCurveData {keys [<at 0 1>]}}
  }
 }
 {
  id=Enveloper_2
  children [Enveloper_1]
  properties {
   bLoop=true
   bLoopIndefinitely=true
   loopEnd=20
   pitchCurve {editorCurveData {keys [{time=0}]}}
   volumeCurve {editorCurveData {keys [
      {arriveTangent=-0.01, interpMode=RCIM_Cubic, time=0, value=1}
      <at 10 0.4 RCIM_Cubic>
      {interpMode=RCIM_Cubic, leaveTangent=0.01, time=20, value=1}
     ]}}
  }
 }
 {id=Mixer_0, children [WavePlayer_0], properties {inputVolume [3]}}
 {id=WavePlayer_0, properties {bLooping=true, soundWaveAssetPtr=/Game/Sounds/Creative/Modes/CTF/CTF_FlagLoop_01}}
]
properties {
 attenuationSettings=/Game/Sounds/Artemis/POI/Reality_Bloom_Atlantis/Spots/Spot_Cloth_Flag_Flapping_Smaller_Att
 bOverrideConcurrency=true
 maxDistance=1100
 pitchMultiplier=0.8
 soundClassObject=&Fort_Ambient_Outdoors_Athena
 volumeMultiplier=0.276
 attenuationOverrides {
  bEnableFocusInterpolation=true
  bEnableListenerFocus=true
  bEnableOcclusion=true
  dBAttenuationAtMax=-40
  distanceAlgorithm=Logarithmic
  falloffDistance=1500
  focusAzimuth=15
  focusVolumeAttenuation=1.1
  nonFocusAzimuth=45
  nonFocusVolumeAttenuation=0.4
  nonSpatializedRadiusStart=100
  occlusionTraceChannel=ECC_GameTraceChannel11
  occlusionVolumeAttenuation=0.7
  stereoSpread=2000
  attenuationShapeExtents {x=100}
 }
 concurrencyOverrides {maxCount=2, voiceStealReleaseTime=1}
 soundSubmixSends [{
   minSendLevel=0.2
   soundSubmix=&Fortnite_Reverb_Default
   customSendLevelCurve {editorCurveData {keys [<at 0 0.4>,<at 1 1>]}}
  }]
}
